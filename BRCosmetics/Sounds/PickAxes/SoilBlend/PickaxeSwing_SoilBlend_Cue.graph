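# SoundCue for the swing sound of the SoilBlend cosmetic pickaxe, mixing a close layer and a medium-range layer, each with its own attenuation.
# The close layer switches on the SwingIdx parameter, playing random SoilBlend swings on most branches and louder random long swings on the last.
# The medium layer plays random shared Fire Axe distant swings through a modulator, independent of SwingIdx.
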SoundCue /BRCosmetics/Sounds/PickAxes/SoilBlend/PickaxeSwing_SoilBlend_Cue
root=Mixer_0
nodes [
 {id=Attenuation_1, children [Modulator_1], properties {attenuationSettings=/Game/Sounds/Attenuation/Fort_Weapon_Pickaxe_Medium_Att}}
 {id=Attenuation_2, children [Modulator_2], properties {attenuationSettings=/Game/Sounds/Attenuation/Fort_Weapon_Pickaxe_Close_Att}}
 {id=Mixer_0, children [Attenuation_2, Attenuation_1], properties {inputVolume [1,1]}}
 {id=Modulator_1, children [Random_1]}
 {id=Modulator_2, children [Switch_0]}
 {id=Modulator_3, children [Random_3], properties {volumeMax=1.25, volumeMin=1.15}}
 {
  id=Random_1
  children [WavePlayer_6, WavePlayer_7, WavePlayer_8, WavePlayer_9, WavePlayer_10, WavePlayer_11]
  properties {weights [1,1,1,1,1,1]}
 }
 {
  id=Random_2
  children [WavePlayer_15, WavePlayer_16, WavePlayer_17, WavePlayer_18, WavePlayer_19, WavePlayer_20]
  properties {weights [1,1,1,1,1,1]}
 }
 {id=Random_3, children [WavePlayer_21, WavePlayer_22, WavePlayer_23], properties {weights [1,1,1]}}
 {id=Switch_0, children [Random_2,Random_2,Random_2,Random_2,Modulator_3], properties {intParameterName=SwingIdx}}
 {id=WavePlayer_10, properties {soundWaveAssetPtr=/Game/Athena/Sounds/Weapons/Fire_Axe_Swing_Distant_05}}
 {id=WavePlayer_11, properties {soundWaveAssetPtr=/Game/Athena/Sounds/Weapons/Fire_Axe_Swing_Distant_06}}
 {id=WavePlayer_15, properties {soundWaveAssetPtr=&/IP_Pickaxe_SoilBlend_Swing_01}}
 {id=WavePlayer_16, properties {soundWaveAssetPtr=&/IP_Pickaxe_SoilBlend_Swing_02}}
 {id=WavePlayer_17, properties {soundWaveAssetPtr=&/IP_Pickaxe_SoilBlend_Swing_03}}
 {id=WavePlayer_18, properties {soundWaveAssetPtr=&/IP_Pickaxe_SoilBlend_Swing_04}}
 {id=WavePlayer_19, properties {soundWaveAssetPtr=&/IP_Pickaxe_SoilBlend_Swing_05}}
 {id=WavePlayer_20, properties {soundWaveAssetPtr=&/IP_Pickaxe_SoilBlend_Swing_06}}
 {id=WavePlayer_21, properties {soundWaveAssetPtr=&/IP_Pickaxe_SoilBlend_Swing_Long_01}}
 {id=WavePlayer_22, properties {soundWaveAssetPtr=&/IP_Pickaxe_SoilBlend_Swing_Long_02}}
 {id=WavePlayer_23, properties {soundWaveAssetPtr=&/IP_Pickaxe_SoilBlend_Swing_Long_03}}
 {id=WavePlayer_6, properties {soundWaveAssetPtr=/Game/Athena/Sounds/Weapons/Fire_Axe_Swing_Distant_01}}
 {id=WavePlayer_7, properties {soundWaveAssetPtr=/Game/Athena/Sounds/Weapons/Fire_Axe_Swing_Distant_02}}
 {id=WavePlayer_8, properties {soundWaveAssetPtr=/Game/Athena/Sounds/Weapons/Fire_Axe_Swing_Distant_03}}
 {id=WavePlayer_9, properties {soundWaveAssetPtr=/Game/Athena/Sounds/Weapons/Fire_Axe_Swing_Distant_04}}
]
properties {
 audioPropertiesSheet=/BRCosmetics/Sounds/Settings/PropertySheets/Pickaxe/APS_Cosmetic_Pickaxe_Swing
 maxDistance=3000
 soundClassObject=&Fort_WeaponMelee_PSM
 volumeMultiplier=0.17
 concurrencySet [&Player_Melee_Weapon_Conc]
}
